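# Translucent unlit Niagara sprite and mesh particle material for the WinterWinds emote's decal-like effect.
# Panned Perlin noise, tinted by particle color, feeds emissive adjustments and also distorts UVs for a sine-based soft spot mask.
# Opacity combines noise, mask and particle alpha with an inverted depth fade inside a second one, so it shows only near intersecting geometry.
# An optional time-driven sine blink and an opacity scalar modulate that alpha.
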Material /BRCosmetics/Animation/Game/MainPlayer/Emotes/WinterWinds/FX/M_WinterWinds_Decal
graph {
 properties {
  bUsedWithNiagaraMeshParticles=true
  bUsedWithNiagaraSprites=true
  blendMode=BLEND_Translucent
  pixelDepthOffsetMode=PDOM_Legacy
  shadingModel=MSM_Unlit
  usedShadingModels=Unlit
 }
 nodes [
  n1 = TextureCoordinate() {uTiling=1.1, vTiling=1.13}
  n2 = ScalarParameter() {defaultValue=0.7415, parameterName=Scale MacroUVs}
  n3 = VectorParameter() {parameterName=MacroUV Panner Speed, defaultValue=<rgba 0.112 -2 0 0>}
  n4 = TextureSample(UVs=Panner(Coordinate=Multiply(A=@n1, B=@n2), Speed=ComponentMask(None=@n3.RGB) {g=true, r=true})) {
   samplerType=SAMPLERTYPE_LinearColor
   texture=/BRCosmetics/Characters/Player/Male/Medium/Bodies/M_MED_HerbHutch_Breath/Skins/Superlevels/Textures/T_SL_S31_PerlinRGB
  }
  n5 = Saturate(None=Power(Base=@n4.R) {constExponent=1})
  n6 = ParticleColor()
  n7 = Multiply(A=@n5, B=@n6.RGB)
  n8 = VectorParameter() {parameterName=EdgeGlow, defaultValue=<rgba 0.5208 2 1.8544 1>}
  n9 = Add(A=@n5) {constB=0}
  Add(A=@n7, B=Multiply(A=Multiply(A=@n8.RGB, B=@n6.RGB), B=@n9))
  MaterialFunctionCall(Emissive=@n7) {materialFunction=&MF_FX_EmissiveAdjustments}
  n10 = Add(A=Multiply(A=@n4.R) {constB=0.12}, B=Add(A=TextureCoordinate(), B=Constant2Vector() {g=-0.088, r=-0.064}))
  n11 = ComponentMask(None=@n10) {r=true}
  n12 = ComponentMask(None=@n10) {g=true}
  n13 = Power(Base=Multiply(A=Sine(None=@n11) {period=2}, B=Sine(None=@n12) {period=2})) {constExponent=3}
  n14 = DepthFade(FadeDistance=ScalarParameter() {defaultValue=35, parameterName=VisibleDepthDistance}) {fadeDistanceDefault=80}
  n15 = Multiply(
   A=Multiply(A=Multiply(A=@n6.A, B=@n9), B=Saturate(None=@n13))
   B=DepthFade(Opacity=OneMinus(None=@n14), FadeDistance=ScalarParameter() {defaultValue=15, parameterName=DepthFade}) {fadeDistanceDefault=15}
  )
  n16 = Time()
  n17 = ScalarParameter() {defaultValue=1, group=Blinking, parameterName=Time}
  n18 = Multiply(A=Sine(None=Multiply(A=@n16, B=@n17)), B=Constant() {r=0.5})
  n19 = Constant() {r=0.5}
  n20 = Clamp(None=Add(A=Add(A=@n18, B=@n19), B=ScalarParameter() {group=Blinking, parameterName=Force Max Sine}) {constA=1})
  Multiply(
   A=LinearInterpolate(A=@n15, B=Multiply(A=@n20, B=@n15), Alpha=ScalarParameter() {group=Blinking, parameterName=Use Blinking})
   B=ScalarParameter() {defaultValue=1, group=Blinking, parameterName=Opacity}
  )
 ]
}
