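# Translucent unlit two-sided Niagara sprite material, named as a side-card effect for the WinterWinds emote.
# Particle color feeds an emissive-adjustment function and supplies alpha.
# Two tiled, diagonally panned Perlin samples multiply into contrast-boosted noise, masked by UV sine falloff toward card edges.
# A camera-to-normal dot product fades edge-on cards, with depth fade on the default shading path and skipped on mobile.
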Material /BRCosmetics/Animation/Game/MainPlayer/Emotes/WinterWinds/FX/M_WinterWinds_Emote_SideCard
graph {
 properties {
  bAutomaticallySetUsageInEditor=false
  bUsedWithNiagaraSprites=true
  blendMode=BLEND_Translucent
  shadingModel=MSM_Unlit
  twoSided=true
  usedShadingModels=Unlit
 }
 nodes [
  n1 = ParticleColor()
  MaterialFunctionCall(Emissive=@n1.RGB) {materialFunction=&MF_FX_EmissiveAdjustments}
  n2 = TextureCoordinate()
  n3 = VectorParameter() {parameterName=Tile, defaultValue=<rgba 0.728 0.76 0.6 1.0076>}
  n4 = AppendVector(A=@n3.R, B=@n3.R)
  n5 = VectorParameter() {parameterName=Speed, defaultValue=<rgba 1.3429 0.224 2 -0.216>}
  n6 = TextureSample(UVs=Panner(Coordinate=Multiply(A=@n2, B=@n4), Speed=AppendVector(A=@n5.R, B=@n5.R))) {
   samplerType=SAMPLERTYPE_LinearColor
   texture=/BRCosmetics/Characters/Player/Male/Medium/Bodies/M_MED_HerbHutch_Breath/Skins/Superlevels/Textures/T_SL_S31_PerlinRGB
  }
  n7 = AppendVector(A=@n3.B, B=@n3.B)
  n8 = TextureSample(UVs=Panner(Coordinate=Multiply(A=@n2, B=@n7), Speed=AppendVector(A=@n5.B, B=@n5.B))) {
   samplerType=SAMPLERTYPE_LinearColor
   texture=&T_SL_S31_PerlinRGB
  }
  n9 = Multiply(A=Power(Base=Multiply(A=@n6.B, B=@n8.B)) {constExponent=1.6}) {constB=1.5}
  n10 = TextureCoordinate()
  n11 = Sine(None=ComponentMask(None=@n10) {r=true}) {period=2}
  n12 = Multiply(A=Saturate(None=Power(Base=@n11) {constExponent=1.5}), B=Sine(None=ComponentMask(None=@n10) {g=true}) {period=2})
  n13 = DotProduct(A=CameraVectorWS(), B=VertexNormalWS())
  n14 = Multiply(A=Multiply(A=@n1.A, B=Multiply(A=@n9, B=@n12)), B=Saturate(None=Abs(None=@n13)))
  ShadingPathSwitch(Default=DepthFade(Opacity=@n14), Mobile=@n14)
 ]
}
