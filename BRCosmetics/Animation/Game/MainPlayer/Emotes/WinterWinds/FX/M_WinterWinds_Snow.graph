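# Translucent, two-sided unlit material for the snow effect of the WinterWinds emote, built for Niagara mesh particles and ribbons.
# Panning Perlin noise, offset and speed-scaled by dynamic parameters, blends two colors that are tinted by particle color and fed through an emissive-adjustment function.
# Noise plus particle alpha is shaped by a UV1 edge-fade mask, then rounded into hard-edged, erodible shapes scaled by dynamic Opacity.
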Material /BRCosmetics/Animation/Game/MainPlayer/Emotes/WinterWinds/FX/M_WinterWinds_Snow
graph {
 properties {
  bUsedWithNiagaraMeshParticles=true
  bUsedWithNiagaraRibbons=true
  blendMode=BLEND_Translucent
  pixelDepthOffsetMode=PDOM_Legacy
  shadingModel=MSM_Unlit
  twoSided=true
  usedShadingModels=Unlit
 }
 nodes [
  n1 = VectorParameter() {parameterName=Color1, defaultValue=<rgba 1 1 1 0>}
  n2 = VectorParameter() {parameterName=Color2, defaultValue=<rgba 0.1364 0.2767 0.5573 1>}
  n3 = TextureCoordinate() {vTiling=0.72}
  n4 = DynamicParameter() {paramNames [RandX,RandY,PannerMult,Opacity]}
  n5 = AppendVector(A=@n4.RandX, B=@n4.RandX)
  n6 = Constant2Vector() {r=-0.344}
  n7 = TextureSampleParameter2D(UVs=Panner(Coordinate=Add(A=@n3, B=@n5), Speed=Multiply(A=@n6, B=@n4.PannerMult)) {speedX=-0.344}) {
   parameterName=Noise
   samplerType=SAMPLERTYPE_LinearGrayscale
   texture=&T_SimplePerlin_Greyscale
  }
  n8 = ParticleColor()
  MaterialFunctionCall(Emissive=Multiply(A=LinearInterpolate(A=@n1.RGB, B=@n2.RGB, Alpha=@n7.R), B=@n8.RGB)) {materialFunction=&MF_FX_EmissiveAdjustments}
  n9 = TextureCoordinate() {coordinateIndex=1}
  n10 = ComponentMask(None=@n9) {r=true}
  n11 = Multiply(A=OneMinus(None=@n10), B=Power(Base=@n10) {constExponent=0.5})
  n12 = Sine(None=ComponentMask(None=@n9) {g=true}) {period=2}
  n13 = Saturate(None=Multiply(A=Multiply(A=@n11, B=@n12)) {constB=3})
  n14 = Multiply(A=Saturate(None=Add(A=@n7.R, B=@n8.A)), B=@n13)
  n15 = Round(None=@n14)
  Multiply(A=@n14, B=@n15)
  Multiply(A=@n15, B=@n4.Opacity)
  Round(None=@n13)
  Saturate(None=Multiply(A=Power(Base=@n14) {constExponent=6}) {constB=20})
  ScalarParameter() {defaultValue=0.8, parameterName=Opacity}
  Sine(None=@n10) {period=2}
 ]
}
