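# Niagara system for the WinterWinds emote that blows snow and wind across the player character.
# Flake, puff, chunk and smoke sprites spawn in offset boxes and stream along negative Y, mostly with curl noise.
# BodyWind pins eight flickering meshes to skeleton bones by spawn group, while SideCard and Decal add persistent sprite cards, the latter gated by owner speed.
# Front-end, platform and time-of-day switches adjust opacity, tint and density.
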NiagaraSystem /BRCosmetics/Animation/Game/MainPlayer/Emotes/WinterWinds/FX/NS_WinterWinds_Emote
dependencies {
 usedDataInterfaces [
  &NiagaraDataInterfaceCurve
  &NiagaraDataInterfaceSpriteRendererInfo
  &NiagaraDataInterfaceColorCurve
  &NiagaraDataInterfacePlatformSet
  &NiagaraDataInterfaceVector2DCurve
  &NiagaraDataInterfaceMeshRendererInfo
  &NiagaraDataInterfaceVectorCurve
  &NiagaraDataInterfaceSkeletalMesh
 ]
 usedDynamicInputs [&SimulationPosition]
 usedModules [
  &self:SystemSpawnScript.SystemScriptSource.SystemScriptGraph.NiagaraNodeAssignment_0.SetVariables_#1
  &SystemState
 ]
}
emitters {
 BodyWind {
  emitterUpdateScript [
   &EmitterState(
    Inactive Response=::NewEnumerator1
    Life Cycle Mode=::NewEnumerator1
    Loop Behavior=::NewEnumerator1
    Loop Duration Mode=::NewEnumerator1
    UseLoopDelay=false
   )
   &SpawnBurst_Instantaneous(Spawn Count=8, Use Spawn Probability=false, UseLoopCountLimit=false)
   &SpawnRate(Spawn Probability=0.6, SpawnRate=12, Use Spawn Probability=true) {enabled=false}
  ]
  particleSpawnScript [
   &FN_SetSpawnGroup(Number of Spawn Groups=8)
   &InitializeParticle(
    Color=&LerpLinearColors(
     StartColor=<rgba 1.3854 1.6191 2 1>
     EndColor=<rgba 1.1083 1.2953 1.6 1>
     LerpFactor=&FN_FrontEndFloatModifier(Gameplay=0, Front End=1)
    )
    Lifetime Max=0.6
    Lifetime Min=0.45
    Lifetime Mode=::NewEnumerator1
    Mesh Scale=&Multiply_Vector(
     A=<0.9 0.01 0.01>
     B=&VectorFromFloat(Value=&MergeHelper_FloatSwitch(0=0.6, 1=0.6, 2=0.7, 3=0.7, 4=0.9, 5=0.9, 6=0.8, 7=0.8, Index=&SpawnGroup(), Num=::NewEnumerator6))
    )
    Mesh Scale Mode=::NewEnumerator4
    UsePositionOffset=false
    Write Material Random=true
   )
   &InitialMeshOrientation(Mesh Axis To Orient=<0 1 0>, Mesh Orientation Mode=::NewEnumerator2, Use Rotation Vector=true)
  ]
  particleUpdateScript [
   &ParticleState(Kill Particles When Lifetime Has Elapsed=false, Let Infinitely Lived Particles Die When Emitter Deactivates=true)
   &ScaleMeshSize(Scale Factor=&VectorFromFloat(Value=&Waveform(
      Global Input Position=&RandomRangeFloat(Minimum=0, Maximum=1)
      "Waveform [1]"=::NewEnumerator10
      "[1] Frequency"=5
      "[1] Amplitude Min/Max"=<0.8 1.7>
     ))) {module=ScaleMeshSize001}
   &ScaleMeshSize(Scale Factor=&VectorFromCurve(
     VectorCurve=DataInterface() {propertyValues {XCurve {keys [<at 0 0 RCIM_Cubic>,<at 1 1>]}, YCurve {keys [<at 0 1>,<at 1 1>]}, ZCurve {keys [<at 0 1>,<at 1 1>]}}}
     CurveIndex=$System.Age
    ))
   &SkeletalMeshLocation(
    Bone / Socket Index=&Add_Integer(A=2, B=&SpawnGroup(Max Spawn Group=15))
    Bone Sampling Mode=::NewEnumerator4
    MaskSpawnGroup=false
    Orientation Sampling=::NewEnumerator1
    Sampled Position Offset=&MergeHelper_VectorSwitch(
     0=<0 0 0>
     1=<0 0 0>
     2=<0 0 2>
     3=<0 0 -2>
     4=<0 0 -6>
     5=<0 0 6>
     6=<0 0 -3>
     7=<0 0 3>
     8=<0 0 0>
     Index=&SpawnGroup()
     Num=::NewEnumerator7
    )
    Skeletal Mesh=$System.Skeletal Mesh
   )
   &DynamicMaterialParameters(
    Index 0 Param 1=&Multiply_Float(
     A=&MergeHelper_FloatSwitch(0=0.5, 1=0.5, 2=0.6, 3=0.6, 4=0.7, 5=0.7, 6=0.6, 7=0.6, Index=&SpawnGroup(), Num=::NewEnumerator6)
     B=&FloatFromCurve(FloatCurve=DataInterface() {propertyValues {Curve {keys [<at 0 0 RCIM_Cubic>,<at 1 1>]}}}, CurveIndex=$System.Age)
    )
    Index 0 Param 2=&RandomRangeFloat(Minimum=3, Maximum=3.3)
    Index 0 Param 4=&RandomRangeFloat(Minimum=0, Maximum=1)
    Index 2 Param 2=13
    Index 2 Param 3=1.3214
    Module.Index0Param0Write=true
    Module.Index0Param1Write=true
    Module.Index0Param2Write=true
    Module.Index0Param3Write=true
    Module.Index2Param0Write=true
    Module.Index2Param1Write=true
    Module.Index2Param2Write=true
    Module.Index2Param3Write=true
    Write Parameter Index 2=true
   )
   &ScaleColor(
    Scale Alpha=&FN_FrontEndFloatModifier(Gameplay=0.17, Front End=0.12)
    Scale RGB=&MakeVectorFromLinearColorRGB(Color=&FNTimeOfDayColor(Night Color=<rgba 0.2598 0.3744 1 1>, Morning Color=<rgba 0.9 0.9 0.9 1>, Day Color=<rgba 1 1 1 1>, Evening Color=<rgba 0.8 0.8 0.8 1>))
    ScaleA=true
    ScaleRGB=true
   ) {module=ScaleColor001}
  ]
  rendererClasses [&NiagaraMeshRendererProperties]
  renderers [{rendererClass=&NiagaraMeshRendererProperties, rendererIndex=0}]
 }
 Decal {
  emitterUpdateScript [
   &EmitterState(
    Inactive Response=::NewEnumerator1
    Life Cycle Mode=::NewEnumerator1
    Loop Behavior=::NewEnumerator1
    Loop Duration Mode=::NewEnumerator1
    UseLoopDelay=false
   )
   &SpawnBurst_Instantaneous(Use Spawn Probability=false, UseLoopCountLimit=false) {module=SpawnBurst_Instantaneous001}
  ]
  particleSpawnScript [
   &InitializeParticle(
    Color=<rgba 0.8958 0.9088 1 0.387>
    Lifetime=0.5
    Position Offset=<0 -20 5>
    Sprite Size=<200 470>
    Sprite Size Mode=::NewEnumerator4
    UsePositionOffset=true
    Write Material Random=true
   ) {module=InitializeParticle001}
   &SpriteFacingAndAlignment(Sprite Facing=<0 0 1>, WriteAlignment=true, WriteFacing=true) {module=SpriteFacingAndAlignment001}
   &CameraOffset(Camera Offset Amount=11) {enabled=false}
  ]
  particleUpdateScript [
   &ParticleState(Kill Particles When Lifetime Has Elapsed=false)
   &ScaleSpriteSize(Uniform Curve Sprite Scale=DataInterface() {propertyValues {Curve {keys [
        {arriveTangent=4.1785, interpMode=RCIM_Cubic, leaveTangent=4.1785, tangentMode=RCTM_User, time=0, value=0}
        {arriveTangent=1.1179, interpMode=RCIM_Cubic, leaveTangent=1.1179, tangentMode=RCTM_User, time=0.2042, value=0.6228}
        {arriveTangent=0.0224, interpMode=RCIM_Cubic, leaveTangent=0.0224, tangentMode=RCTM_User, time=1, value=1}
       ]}}})
   &ScaleColor(
    Scale Alpha=&FloatFromCurve(
     FloatCurve=DataInterface() {propertyValues {Curve {keys [
         {arriveTangent=4.1785, interpMode=RCIM_Cubic, leaveTangent=4.1785, tangentMode=RCTM_User, time=0, value=0}
         {arriveTangent=1.5009, interpMode=RCIM_Cubic, leaveTangent=1.5009, tangentMode=RCTM_User, time=0.2, value=0.55}
         {arriveTangent=0.0224, interpMode=RCIM_Cubic, leaveTangent=0.0224, tangentMode=RCTM_User, time=1, value=1}
        ]}}}
     Scale Curve=&Multiply_Float(
      A=&FN_FrontEndFloatModifier(Gameplay=0.8, Front End=0.45)
      B=&MakeCustomFloatFromBool(
       Bool=&SetBoolByFloatComparison(A=&VectorLength(Vector=$Engine.Owner.Velocity), B=150)
       True Float=0
       False Float=1
      )
     )
    )
    Scale RGB=&MakeVectorFromLinearColorRGB(Color=&FNTimeOfDayColor(Night Color=<rgba 0.3356 0.4595 1 1>, Morning Color=<rgba 0.9 0.9 0.9 1>, Day Color=<rgba 1 1 1 1>, Evening Color=<rgba 0.8 0.8 0.8 1>))
    ScaleA=true
    ScaleRGB=true
   )
  ]
  rendererClasses [&NiagaraSpriteRendererProperties]
  renderers [{rendererClass=&NiagaraSpriteRendererProperties}]
 }
 Idle_Flakes {
  emitterUpdateScript [
   &EmitterState()
   &SpawnRate(Spawn Probability=0.8, SpawnRate=28, Use Spawn Probability=true)
  ]
  particleSpawnScript [
   &InitializeParticle(
    Lifetime=&RandomRangeFloat(Minimum=0.6, Maximum=0.8)
    Mass Mode=::NewEnumerator1
    Sprite Rotation Angle=&RandomRangeFloat(Minimum=0, Maximum=1)
    Sprite Rotation Mode=::NewEnumerator0
    Sprite Size=&Vector2DFromFloat(Value=&RandomRangeFloat(Minimum=2, Maximum=5))
    Sprite Size Mode=::NewEnumerator4
    UsePositionOffset=false
    Write Material Random=true
   ) {module=InitializeParticle001}
   &Color(Color=<rgba 0.4 0.59 1 1>, Scale Alpha=$Particles.MaterialRandom)
   &ShapeLocation(
    Box Size=<150 60 200>
    MaskSpawnGroup=false
    Offset=<0 200 80>
    Offset Coordinate Space=::Local
    Offset Mode=::NewEnumerator0
    Override Seed=false
    Shape Primitive=::NewEnumerator2
    Surface Only=false
    UseOwnerScale=true
   )
   &AddVelocity(Velocity=&RandomRangeVector(Minimum=<-50 -600 -60>, Maximum=<50 -800 -20>))
   &CameraOffset(Camera Offset Amount=-20)
  ]
  particleUpdateScript [
   &ParticleState()
   &CurlNoiseForce(
    Bias Noise=false
    Noise Frequency=14
    Noise Strength=&RandomRangeFloat(Minimum=2000, Maximum=5000)
    Pan Noise=false
   )
   &SolveForcesAndVelocity(Clamp Velocity=false, Limit Acceleration=false)
   &SpriteRotationRate(Rotation Rate=&FloatFromCurve(
     FloatCurve=DataInterface() {propertyValues {Curve {keys [<at 0 0 RCIM_Cubic>,<at 0.3 1 RCIM_Cubic>,<at 1 0>]}}}
     Scale Curve=&RandomRangeFloat(Minimum=-100, Maximum=100)
    ))
   &ScaleColor(
    Scale Alpha=&FloatFromCurve(FloatCurve=DataInterface() {propertyValues {Curve {keys [<at 0.6 1>,<at 1 0>]}}})
    ScaleA=true
    ScaleRGB=true
   )
   &ScaleSpriteSize(Scale Factor=&Vector2DFromFloat(Value=&FloatFromCurve(
      FloatCurve=DataInterface() {propertyValues {Curve {keys [
          {arriveTangent=0.01, interpMode=RCIM_Cubic, time=0, value=0}
          {arriveTangent=2, interpMode=RCIM_Cubic, leaveTangent=2, time=0.1357, value=0.7792}
          {arriveTangent=-0.9016, interpMode=RCIM_Cubic, leaveTangent=-0.9016, time=0.5, value=1}
          <at 1 0 RCIM_Cubic>
         ]}}}
      Scale Curve=1.5
     )))
  ]
  rendererClasses [&NiagaraSpriteRendererProperties]
  renderers [{rendererClass=&NiagaraSpriteRendererProperties}]
 }
 Idle_Puffs {
  emitterUpdateScript [
   &EmitterState()
   &SpawnRate(SpawnRate=8, Use Spawn Probability=false)
  ]
  particleSpawnScript [
   &InitializeParticle(
    Color=<rgba 1 1 1 0.09>
    Lifetime=&RandomRangeFloat(Minimum=0.45, Maximum=0.5)
    Mass Mode=::NewEnumerator1
    Sprite Rotation Angle=&RandomRangeFloat(Minimum=0, Maximum=360)
    Sprite Rotation Mode=::NewEnumerator0
    Sprite Size=&Vector2DFromFloat(Value=&RandomRangeFloat(Minimum=10, Maximum=20))
    Sprite Size Mode=::NewEnumerator4
    UsePositionOffset=false
    Write Material Random=true
   ) {module=InitializeParticle001}
   &CameraOffset(Camera Offset Amount=-20)
   &ShapeLocation(
    Box Size=<60 30 50>
    MaskSpawnGroup=false
    Offset=<0 230 100>
    Offset Coordinate Space=::Local
    Offset Mode=::NewEnumerator0
    Override Seed=false
    Shape Primitive=::NewEnumerator2
    Surface Only=false
    UseOwnerScale=true
   )
   &AddVelocity(Velocity=&RandomRangeVector(Minimum=<-50 -700 -100>, Maximum=<50 -1000 -60>))
  ]
  particleUpdateScript [
   &ParticleState()
   &SolveForcesAndVelocity(Clamp Velocity=false, Limit Acceleration=false)
   &SubUVAnimation(
    "Sprite Renderer "=DataInterface() {propertyValues {SpriteRenderer=&self:NiagaraEmitter_6.NiagaraSpriteRendererProperties_1}}
    SubUV Lookup Index Scale=0.4
    UseEndFrame=false
    UseStartFrame=false
   ) {module=SubUVAnimation001}
   &SpriteRotationRate(Rotation Rate=&FloatFromCurve(
     FloatCurve=DataInterface() {propertyValues {Curve {keys [<at 0 0 RCIM_Cubic>,<at 0.3 1 RCIM_Cubic>,<at 1 0>]}}}
     Scale Curve=&RandomRangeFloat(Minimum=-20, Maximum=20)
    ))
   &ScaleSpriteSize(
    Uniform Curve Scale=18
    Uniform Curve Sprite Scale=DataInterface() {propertyValues {Curve {keys [
        {arriveTangent=1.2925, interpMode=RCIM_Cubic, leaveTangent=1.2925, tangentMode=RCTM_User, time=0, value=0.3}
        {tangentMode=RCTM_User, time=1, value=0.8685}
       ]}}}
   )
   &ScaleColor(
    Scale Mode=::NewEnumerator1
    Scale RGBA=&MakeVector4_FromLinearColor(LinearColor=&ColorFromCurve(ColorCurve=DataInterface() {propertyValues {
        AlphaCurve {keys [
          <at 0 0 RCIM_Cubic>
          {arriveTangent=0.7867, interpMode=RCIM_Cubic, leaveTangent=0.7867, time=0.3694, value=0.65}
          {arriveTangent=-1.0308, interpMode=RCIM_Cubic, leaveTangent=-1.0308, time=0.6355, value=0.5}
          <at 1 0>
          <at 1 0>
         ]}
        BlueCurve {keys [<at 0 1 RCIM_Cubic>,<at 1 1 RCIM_Cubic>]}
        GreenCurve {keys [<at 0 0.7578 RCIM_Cubic>,<at 1 0.8323 RCIM_Cubic>]}
        RedCurve {keys [<at 0 0.6771 RCIM_Cubic>,<at 1 0.7604 RCIM_Cubic>]}
       }}))
    ScaleRGBA=true
   ) {module=ScaleColor001}
   &ScaleColor(
    Scale Mode=::NewEnumerator1
    Scale RGBA=&MakeVector4_FromLinearColor(LinearColor=&FNTimeOfDayColor(Night Color=<rgba 0.3333 0.4219 1 1>, Morning Color=<rgba 0.9 0.9 0.9 1>, Day Color=<rgba 1 1 1 1>, Evening Color=<rgba 0.8 0.8 0.8 1>))
    ScaleRGBA=true
   )
   &ScaleColor(
    Scale Alpha=&FN_PickFloatDependOnPlatform(TargetPlatform=DataInterface() {propertyValues {Platforms {qualityLevelMask=-29}}}, TargetPlatformActive=0.35, TargetPlatformNoActive=1)
    ScaleA=true
    ScaleRGB=true
   ) {module=ScaleColor002}
  ]
  rendererClasses [&NiagaraSpriteRendererProperties]
  renderers [{rendererClass=&NiagaraSpriteRendererProperties}]
 }
 SideCard {
  emitterUpdateScript [
   &EmitterState(
    Inactive Response=::NewEnumerator1
    Life Cycle Mode=::NewEnumerator1
    Loop Behavior=::NewEnumerator1
    Loop Duration Mode=::NewEnumerator1
    UseLoopDelay=false
   )
   &SpawnBurst_Instantaneous(Use Spawn Probability=false, UseLoopCountLimit=false)
   &SpawnBurst_Instantaneous(Spawn Group=1, Use Spawn Probability=false, UseLoopCountLimit=false) {module=SpawnBurst_Instantaneous001}
  ]
  particleSpawnScript [
   &InitializeParticle(
    Color=&MaskLinearColorBySpawnGroup(Masked Linear Color=<rgba 0.7448 0.8941 1 0.427>, Passthrough Linear Color=<rgba 0.7448 0.8941 1 0.251>)
    Position Offset=&Add_Vector(
     A=&MaskVectorBySpawnGroup(Masked Vector=<0 20 63>, Passthrough Vector=<0 20 70>)
     B=&MakeVector(
      X=0
      Y=0
      Z=&FN_PickFloatDependOnPlatform(TargetPlatform=DataInterface() {propertyValues {Platforms {qualityLevelMask=-29}}}, TargetPlatformActive=14, TargetPlatformNoActive=0)
     )
    )
    Sprite Size=&MaskVector2dBySpawnGroup(
     Masked Vector2D=&Multiply_Vector2D(
      A=<330 215>
      B=&MakeVector2D(
       X=1
       Y=&FN_PickFloatDependOnPlatform(TargetPlatform=DataInterface() {propertyValues {Platforms {qualityLevelMask=-29}}}, TargetPlatformActive=0.7, TargetPlatformNoActive=1)
      )
     )
     Passthrough Vector2D=<330 150>
     Spawn Group Mask=0
    )
    Sprite Size Mode=::NewEnumerator4
    UsePositionOffset=true
    Write Material Random=true
   )
   &SpriteFacingAndAlignment(
    Sprite Alignment=&MaskVectorBySpawnGroup(Masked Vector=<0 0 1>, Passthrough Vector=<1 0 0>)
    Sprite Facing=&MaskVectorBySpawnGroup(Masked Vector=<-1 0 0>, Passthrough Vector=<0 0 1>)
    WriteAlignment=true
    WriteFacing=true
   )
  ]
  particleUpdateScript [
   &ParticleState(Kill Particles When Lifetime Has Elapsed=false, Let Infinitely Lived Particles Die When Emitter Deactivates=true)
   &ScaleColor(
    Scale Alpha=&FloatFromCurve(
     FloatCurve=DataInterface() {propertyValues {Curve {keys [<at 0.2097 0 RCIM_Cubic>,<at 0.5221 1>]}}}
     CurveIndex=$System.Age
     Scale Curve=&FN_FrontEndFloatModifier(
      Gameplay=&FN_PickFloatDependOnPlatform(TargetPlatform=DataInterface() {propertyValues {Platforms {qualityLevelMask=-29}}}, TargetPlatformActive=0.6, TargetPlatformNoActive=1)
      Front End=0.5
     )
    )
    Scale RGB=&MakeVectorFromLinearColorRGB(Color=&FNTimeOfDayColor(Night Color=<rgba 0.4427 0.5589 1 1>, Morning Color=<rgba 0.9 0.9 0.9 1>, Day Color=<rgba 1 1 1 1>, Evening Color=<rgba 0.8 0.8 0.8 1>))
    ScaleA=true
    ScaleRGB=true
   )
  ]
  rendererClasses [&NiagaraSpriteRendererProperties]
  renderers [{rendererClass=&NiagaraSpriteRendererProperties}]
 }
 SnowChunks {
  emitterUpdateScript [
   &EmitterState()
   &SpawnRate(Spawn Probability=0.8, SpawnRate=28, Use Spawn Probability=true)
  ]
  particleSpawnScript [
   &InitializeParticle(
    Color=<rgba 1 1 1 0.247>
    Lifetime=&RandomRangeFloat(Minimum=0.5, Maximum=0.7)
    Sprite Size Mode=::NewEnumerator1
    Uniform Sprite Size Max=6
    Uniform Sprite Size Min=3
    UsePositionOffset=false
    Write Material Random=true
   )
   &ShapeLocation(
    Box Size=<150 60 200>
    MaskSpawnGroup=false
    Offset=<0 200 80>
    Offset Coordinate Space=::Local
    Offset Mode=::NewEnumerator0
    Override Seed=false
    Shape Primitive=::NewEnumerator2
    Surface Only=false
    UseOwnerScale=true
   )
   &AddVelocity(Velocity=&RandomRangeVector(Minimum=<-50 -600 -60>, Maximum=<50 -800 -20>))
  ]
  particleUpdateScript [
   &ParticleState()
   &ScaleSpriteSize(
    Scale Factor=&RandomRangeVector2D(Minimum=<0.2 0.3>, Maximum=<0.5 1.6>)
    Scale Sprite Size Mode=::NewEnumerator2
   )
   &CurlNoiseForce(
    Bias Noise=false
    Noise Frequency=14
    Noise Strength=&RandomRangeFloat(Minimum=2000, Maximum=5000)
    Pan Noise=false
   )
   &SolveForcesAndVelocity(Clamp Velocity=false, Limit Acceleration=false)
   &ScaleSpriteSizeBySpeed(Max Scale Factor=<1 3>, Min Scale Factor=<0.3 1>, Velocity Threshold=300)
   &ScaleColor(
    Scale Alpha=&FloatFromCurve(
     FloatCurve=DataInterface() {propertyValues {Curve {keys [{interpMode=RCIM_Cubic, tangentMode=RCTM_User, time=0.4, value=1}, {interpMode=RCIM_Cubic, tangentMode=RCTM_User, time=1, value=0}]}}}
     Scale Curve=0.6
    )
    Scale RGB=&MakeVectorFromLinearColorRGB(Color=&FNTimeOfDayColor(Night Color=<rgba 1 0 0.3847 1>, Morning Color=<rgba 1 1 1 1>, Day Color=<rgba 1 1 1 1>, Evening Color=<rgba 1 1 1 1>))
    ScaleA=true
    ScaleRGB=true
   )
  ]
  rendererClasses [&NiagaraSpriteRendererProperties]
  renderers [{rendererClass=&NiagaraSpriteRendererProperties}]
 }
 Snow_Smoke {
  emitterUpdateScript [
   &EmitterState()
   &SpawnRate(SpawnRate=&FN_FrontEndFloatModifier(Gameplay=8, Front End=0), Use Spawn Probability=false)
  ]
  particleSpawnScript [
   &InitializeParticle(
    Color=<rgba 0.5885 0.745 1 1>
    Lifetime=&UniformRangedFloat(Minimum=0.4, Maximum=0.45)
    Mass=10
    Mass Mode=::NewEnumerator1
    Sprite Size Mode=::NewEnumerator1
    Uniform Sprite Size Max=50
    Uniform Sprite Size Min=30
    UsePositionOffset=false
    Write Material Random=true
   ) {module=InitializeParticle001}
   &ShapeLocation(
    Box Size=<70 40 80>
    MaskSpawnGroup=false
    Offset=<0 200 80>
    Offset Coordinate Space=::Local
    Offset Mode=::NewEnumerator0
    Override Seed=false
    Shape Primitive=::NewEnumerator2
    Surface Only=false
    UseOwnerScale=true
   )
   &AddVelocity(Velocity=&RandomRangeVector(Minimum=<-50 -700 -80>, Maximum=<50 -1000 -40>))
  ]
  particleUpdateScript [
   &ParticleState()
   &ScaleSpriteSize(
    Scale Factor=&MaskVector2dBySpawnGroup(
     Masked Vector2D=&Vector2DFromCurve(
      Vector2Curve=DataInterface() {propertyValues {
        XCurve {keys [<at 0 1>,<at 1 1>]}
        YCurve {keys [{arriveTangent=2.898, interpMode=RCIM_Cubic, leaveTangent=2.898, tangentMode=RCTM_User, time=0, value=1.6993}, <at 1 3>]}
       }}
      ScaleCurve=&RandomRangeVector2D(Minimum=<0.7 0.6>, Maximum=<1 0.85>)
     )
     Passthrough Vector2D=<3 2>
     Spawn Group Mask=0
    )
    Scale Sprite Size Mode=::NewEnumerator2
   ) {module=ScaleSpriteSize001}
   &ScaleSpriteSize(Uniform Curve Sprite Scale=DataInterface() {propertyValues {Curve {keys [
        {arriveTangent=6.1457, interpMode=RCIM_Cubic, leaveTangent=6.1457, tangentMode=RCTM_User, time=0, value=0.4748}
        {arriveTangent=1.9065, interpMode=RCIM_Cubic, leaveTangent=1.9065, time=0.2546, value=1.2632}
        <at 0.8 2 RCIM_Cubic>
       ]}}})
   &CurlNoiseForce(
    Bias Noise=false
    Noise Frequency=14
    Noise Strength=&RandomRangeFloat(Minimum=0, Maximum=300)
    Pan Noise=false
   )
   &SolveForcesAndVelocity(Clamp Velocity=false, Limit Acceleration=false)
   &ScaleColor(
    Scale Alpha=&FloatFromCurve(
     FloatCurve=DataInterface() {propertyValues {Curve {keys [
         <at 0 0 RCIM_Cubic>
         {interpMode=RCIM_Cubic, tangentMode=RCTM_User, time=0.2, value=0.375}
         {arriveTangent=-1.7261, interpMode=RCIM_Cubic, leaveTangent=-1.7261, tangentMode=RCTM_User, time=1, value=0}
        ]}}}
     Scale Curve=&MaskFloatBySpawnGroup(Masked Float=0.7, Passthrough Float=0.4)
    )
    ScaleA=true
    ScaleRGB=true
   )
   &SpriteRotationRate(Rotation Rate=&RandomRangeFloat(Minimum=-360, Maximum=360))
   &DynamicMaterialParameters(
    Index 0 Param 1=&Add_Float(
     A=&RandomRangeFloat(Minimum=0, Maximum=1)
     B=&FloatFromCurve(FloatCurve=DataInterface() {propertyValues {Curve {keys [{arriveTangent=2.739, time=0, value=0}, {leaveTangent=0.0224, time=1, value=0.5436}]}}})
    )
    Index 0 Param 3=&FloatFromCurve(FloatCurve=DataInterface() {propertyValues {Curve {keys [<at 0.25 1>,<at 1 0>]}}}, Scale Curve=0.1)
    Index 0 Param 4=&FloatFromCurve(FloatCurve=DataInterface() {propertyValues {Curve {keys [<at 0 1>,<at 1 0>]}}})
    Index 2 Param 1=0.8
    Index 2 Param 2=&FloatFromCurve(
     FloatCurve=DataInterface() {propertyValues {Curve {keys [{tangentMode=RCTM_User, time=0, value=0.2}, {tangentMode=RCTM_User, time=1, value=0.8231}]}}}
     Scale Curve=&RandomRangeFloat(Minimum=0.5, Maximum=1)
    )
    Index 2 Param 3=2
    Index 2 Param 4=1
    Module.Index0Param0Write=true
    Module.Index0Param1Write=true
    Module.Index0Param2Write=true
    Module.Index0Param3Write=true
    Module.Index1Param0Write=true
    Module.Index1Param1Write=true
    Module.Index1Param2Write=true
    Module.Index1Param3Write=true
    Module.Index2Param0Write=true
    Module.Index2Param1Write=true
    Module.Index2Param2Write=true
    Module.Index2Param3Write=true
    Write Parameter Index 1=true
    Write Parameter Index 2=true
   )
  ]
  rendererClasses [&NiagaraSpriteRendererProperties]
  renderers [{rendererClass=&NiagaraSpriteRendererProperties}]
 }
}
summary {
 systemName=NS_WinterWinds_Emote
 emitters [
  {emitterName=Idle_Flakes, rendererClasses [&NiagaraSpriteRendererProperties]}
  {emitterName=Idle_Puffs, rendererClasses [&NiagaraSpriteRendererProperties]}
  {emitterName=Snow_Smoke, rendererClasses [&NiagaraSpriteRendererProperties]}
  {emitterName=SnowChunks, rendererClasses [&NiagaraSpriteRendererProperties]}
  {emitterName=BodyWind, rendererClasses [&NiagaraMeshRendererProperties]}
  {emitterName=SideCard, rendererClasses [&NiagaraSpriteRendererProperties]}
  {emitterName=Decal, rendererClasses [&NiagaraSpriteRendererProperties]}
 ]
 userVariables [
  User.bIsFrontEnd {type=NiagaraBool}
  User.bIsFrontEndPreview {type=NiagaraBool}
  User.HeightThreshold {type=NiagaraFloat, defaultValue=3.4, description=""}
  User.IsSkydiving {type=NiagaraBool}
  User.LandBurst {type=NiagaraFloat, defaultValue=0, description=""}
  User.RibbonIDOffset {type=NiagaraInt32}
  User.SprintBurst {type=NiagaraFloat, defaultValue=0, description=""}
 ]
}
