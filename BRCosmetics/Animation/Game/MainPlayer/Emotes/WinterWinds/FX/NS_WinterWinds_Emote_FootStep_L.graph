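# Niagara effect for the left footstep of the WinterWinds emote, producing snow at the foot and a ribbon trail from the skeletal mesh.
# SnowFeet001 uses spawn groups to separate stationary, foot-oriented meshes from small flying chunks, alongside animated puff sprites; User.LeftFoot blends mesh rotation and scale.
# Trail uses FN AnimTrail modules with a fading, narrowing ribbon.
# Both emitters tint by time of day and gate sprite size or trail material on owner speed.
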NiagaraSystem /BRCosmetics/Animation/Game/MainPlayer/Emotes/WinterWinds/FX/NS_WinterWinds_Emote_FootStep_L
dependencies {
 usedDataInterfaces [
  &NiagaraDataInterfaceCurve
  &NiagaraDataInterfaceVectorCurve
  &NiagaraDataInterfaceSpriteRendererInfo
  &NiagaraDataInterfaceColorCurve
  &NiagaraDataInterfaceSkeletalMesh
 ]
 usedDynamicInputs [&MergeHelper_VectorSwitch, &Vector2DFromFloat]
 usedModules [
  &self:SystemSpawnScript.SystemScriptSource.SystemScriptGraph.NiagaraNodeAssignment_0.SetVariables_#1
  &SystemState
 ]
}
emitters {
 SnowFeet001 {
  emitterUpdateScript [
   &EmitterState()
   &SpawnRate(SpawnGroup=1, SpawnRate=3, Use Spawn Probability=false)
   &SpawnBurst_Instantaneous(Use Spawn Probability=false, UseLoopCountLimit=false)
  ]
  particleSpawnScript [
   &InitializeParticle(
    Color=<rgba 0.1 0.0928 0.084 0.2>
    Lifetime=&MaskFloatBySpawnGroup(
     Masked Float=&RandomRangeFloat(Minimum=0.9, Maximum=1)
     Passthrough Float=&RandomRangeFloat(Minimum=0.35, Maximum=0.45)
    )
    Mesh Scale=&MaskVectorBySpawnGroup(
     Masked Vector=<1.1 1.1 1.1>
     Passthrough Vector=&RandomRangeVector(Minimum=<0.02 0.02 0.02>, Maximum=<0.05 0.05 0.05>)
    )
    Mesh Scale Mode=::NewEnumerator4
    Position Offset=<0 2.4 12>
    Position Offset Coordinate Space=::Local
    Sprite Rotation Angle=&RandomRangeFloat(Minimum=0.3, Maximum=0.7)
    Sprite Rotation Mode=::NewEnumerator1
    Sprite Size Mode=::NewEnumerator1
    Uniform Sprite Size Max=44
    Uniform Sprite Size Min=36
    UsePositionOffset=true
    Write Material Random=true
   )
   /SetVariables(
    Particles.MeshIndex1:NiagaraInt32=&RandomRangeInt(Minimum=0, Maximum=3)
    Particles.MeshIndex2:NiagaraInt32=&RandomRangeInt(Minimum=0, Maximum=1)
   ) {isSetParameters=true}
   &InitialMeshOrientation(
    Mesh Orientation Mode=::NewEnumerator2
    Orientation Coordinate Space=::Local
    Rotation=&MaskVectorBySpawnGroup(
     Masked Vector=&MakeVector(
      X=0
      Y=0
      Z=&Lerp_Float(
       A=&RandomRangeFloat(Minimum=0, Maximum=1)
       B=&UniformAOrBFloat(A=0.25, B=0.75, Evaluation Type=::NewEnumerator0, Recalculate Random Each Loop=true, Override Seed=false)
       Alpha=$User.LeftFoot
      )
     )
     Passthrough Vector=&RandomRangeVector(Minimum=<0 0 0>, Maximum=<1 1 1>)
    )
    Use Rotation Vector=true
   )
   &SkeletalMeshLocation(
    Bone Sampling Mode=::NewEnumerator4
    MaskSpawnGroup=false
    Orientation Sampling=::NewEnumerator1
    Sampled Position Offset=<-14 2.3 0>
    Skeletal Mesh=$System.Skeletal Mesh
   )
   &AddVelocity(
    Rotation Coordinate Space=::World
    Velocity=&RandomRangeVector(Minimum=<-100 -100 50>, Maximum=<100 100 100>)
    Velocity Speed Scale=&MaskFloatBySpawnGroup(Masked Float=0, Passthrough Float=1)
   )
   /SetVariables(Particles.VisibilityTag=&SpawnGroup()) {isSetParameters=true}
  ]
  particleUpdateScript [
   &ParticleState()
   &GravityForce(Gravity=&Multiply_Vector(
     A=<0 0 -350.1571>
     B=&VectorFromFloat(Value=&MaskFloatBySpawnGroup(Masked Float=0, Passthrough Float=1))
    ))
   &SolveForcesAndVelocity(Clamp Velocity=false, Limit Acceleration=false)
   &ScaleMeshSize(Scale Factor=&VectorFromFloat(Value=&FloatFromCurve(FloatCurve=DataInterface() {propertyValues {Curve {keys [
          {arriveTangent=8.1506, interpMode=RCIM_Cubic, leaveTangent=8.1506, tangentMode=RCTM_User, time=0, value=0}
          <at 0.12 1>
          {arriveTangent=-1.1364, interpMode=RCIM_Cubic, leaveTangent=-1.1364, time=0.8789, value=1}
          <at 1 0>
         ]}}})))
   &ScaleMeshSize(Scale Factor=&MaskVectorBySpawnGroup(
     Masked Vector=&VectorFromCurve(
      VectorCurve=DataInterface() {propertyValues {XCurve {keys [<at 0 1>,<at 1 1>]}, YCurve {keys [<at 0 1>,<at 1 1>]}, ZCurve {keys [<at 0 1>,<at 1 1.1 RCIM_Cubic>]}}}
      Scale Curve=&Lerp_Vector(A=<1.4 1.4 2>, B=<2.3 1.5 2>, Alpha=$User.LeftFoot)
     )
     Passthrough Vector=<1 1 1>
    )) {module=ScaleMeshSize001}
   &SubUVAnimation(
    "Sprite Renderer "=DataInterface() {propertyValues {SpriteRenderer=&self:SnowFeet001_1.NiagaraSpriteRendererProperties_1}}
    SubUV Lookup Index Scale=0.4
    UseEndFrame=false
    UseStartFrame=false
   ) {module=SubUVAnimation001}
   &ScaleSpriteSize(
    Uniform Curve Scale=&Multiply_Float(
     A=2
     B=&MakeCustomFloatFromBool(
      Bool=&SetBoolByFloatComparison(A=&VectorLength(Vector=$Engine.Owner.Velocity), B=150)
      True Float=0
      False Float=1
     )
    )
    Uniform Curve Sprite Scale=DataInterface() {propertyValues {Curve {keys [
        {arriveTangent=2.6419, interpMode=RCIM_Cubic, leaveTangent=2.6419, tangentMode=RCTM_User, time=0, value=0.3}
        {interpMode=RCIM_Cubic, leaveTangent=-0.6, time=1, value=1}
       ]}}}
   )
   /SetVariables(Particles.SnowSpriteColor:LinearColor=&MultiplyLinearColor(
     Linear Color=&ColorFromCurve(ColorCurve=DataInterface() {propertyValues {
        AlphaCurve {keys [<at 0.1204 0.2 RCIM_Cubic>,<at 1 0 RCIM_Cubic>]}
        BlueCurve {keys [<at 0 1 RCIM_Cubic>,<at 1 1 RCIM_Cubic>]}
        GreenCurve {keys [<at 0 1 RCIM_Cubic>,<at 1 1 RCIM_Cubic>]}
        RedCurve {keys [<at 0 1 RCIM_Cubic>,<at 1 1 RCIM_Cubic>]}
       }})
     Scale Factor=&MakeVector4_FromLinearColor(LinearColor=&FNTimeOfDayColor(Night Color=<rgba 0.3092 0.3925 0.8 1>, Morning Color=<rgba 0.8 0.8 0.8 1>, Day Color=<rgba 1 1 1 1>, Evening Color=<rgba 0.8 0.8 0.8 1>))
    )) {isSetParameters=true}
   /SetVariables(
    Particles.SpriteDynamic:Vector4f=&MakeVector4(
     X=&Add_Float(
      A=&RandomRangeFloat(Minimum=0, Maximum=1)
      B=&FloatFromCurve(FloatCurve=DataInterface() {propertyValues {Curve {keys [{arriveTangent=2.739, time=0, value=0}, {leaveTangent=0.0224, time=1, value=0.5436}]}}})
     )
     Y=0
     Z=&FloatFromCurve(FloatCurve=DataInterface() {propertyValues {Curve {keys [<at 0.25 1>,<at 1 0>]}}}, Scale Curve=0.1)
     W=&FloatFromCurve(FloatCurve=DataInterface() {propertyValues {Curve {keys [<at 0 1>,<at 1 0>]}}})
    )
    Particles.SpritePosition:NiagaraPosition=&AddVectorToPosition(Position=$Particles.Position, Vector=<0 0 -4>)
   ) {isSetParameters=true}
   /SetVariables(Particles.SnowBlobsSize:Vector3f=&VectorFromCurve(
     VectorCurve=DataInterface() {propertyValues {
       XCurve {keys [
         {arriveTangent=3.3816, interpMode=RCIM_Cubic, leaveTangent=3.3816, tangentMode=RCTM_User, time=0, value=0}
         <at 0.5 1 RCIM_Cubic>
         {arriveTangent=-3.4632, interpMode=RCIM_Cubic, leaveTangent=-3.4632, tangentMode=RCTM_User, time=1, value=0}
        ]}
       YCurve {keys [
         {arriveTangent=3.3816, interpMode=RCIM_Cubic, leaveTangent=3.3816, tangentMode=RCTM_User, time=0, value=0}
         <at 0.5 1 RCIM_Cubic>
         {arriveTangent=-3.4632, interpMode=RCIM_Cubic, leaveTangent=-3.4632, tangentMode=RCTM_User, time=1, value=0}
        ]}
       ZCurve {keys [
         {arriveTangent=3.3816, interpMode=RCIM_Cubic, leaveTangent=3.3816, tangentMode=RCTM_User, time=0, value=0}
         <at 0.5 1 RCIM_Cubic>
         {arriveTangent=-3.4632, interpMode=RCIM_Cubic, leaveTangent=-3.4632, tangentMode=RCTM_User, time=1, value=0}
        ]}
      }}
     Scale Curve=&RandomRangeVector(Minimum=<0.015 0.015 0.015>, Maximum=<0.04 0.04 0.04>)
    )) {isSetParameters=true}
  ]
  rendererClasses [&NiagaraSpriteRendererProperties, &NiagaraMeshRendererProperties]
  renderers [
   {rendererClass=&NiagaraSpriteRendererProperties}
   {rendererClass=&NiagaraMeshRendererProperties, rendererIndex=1}
  ]
 }
 Trail {
  emitterUpdateScript [
   &EmitterState()
   &FN_SpawnAnimTrail(Spawn Mode:ENiagaraSkeletalSpawning=::NewEnumerator0)
   /SetVariables(
    Emitter.RandFloat1:NiagaraFloat=&RandomRangeFloat(Minimum=0, Maximum=1)
    Emitter.RandFloat2:NiagaraFloat=&RandomRangeFloat(Minimum=0, Maximum=1)
   ) {isSetParameters=true}
  ]
  particleSpawnScript [
   &FN_InitializeAnimTrailBoneIndecies(Second Bone or Socket Index=0) {module=FN_InitializeAnimTrailBoneIndecies001}
   &FN_InitializeAnimTrail(
    AnimTrailWidth=14
    Color=<rgba 1 1 1 0.7>
    Lifetime=1
    Position Offset=&MakeVector(X=0, Y=0, Z=&Subtract_Float(A=-2.3, B=$User.TrailOffset))
    SkeletalMesh=$System.Skeletal Mesh
   ) {module=FN_InitializeAnimTrail001}
   &DynamicMaterialParameters(
    Index 0 Param 1=$Emitter.RandFloat1
    Index 0 Param 2=$Emitter.RandFloat2
    Index 0 Param 3=0.8
    Index 0 Param 4=&Multiply_Float(
     A=&FN_FrontEndFloatModifier(
      Gameplay=&FNTimeOfDayScalar(Night Scale=0.6, Morning Scale=0.7, Day Scale=0.8, Evening Scale=0.7)
      Front End=0.5
     )
     B=&MakeCustomFloatFromBool(
      Bool=&SetBoolByFloatComparison(A=&VectorLength(Vector=$Engine.Owner.Velocity), B=150)
      True Float=0
      False Float=1
     )
    )
    Module.Index0Param0Write=true
    Module.Index0Param1Write=true
    Module.Index0Param2Write=true
    Module.Index0Param3Write=true
   )
  ]
  particleUpdateScript [
   &FN_UpdateAnimTrail(SkeletalMesh=$System.Skeletal Mesh)
   &ScaleRibbonWidth(Ribbon Width Scale=&FloatFromCurve(FloatCurve=DataInterface() {propertyValues {Curve {keys [
         {interpMode=RCIM_Cubic, tangentMode=RCTM_User, time=0, value=1}
         {interpMode=RCIM_Cubic, tangentMode=RCTM_User, time=0.5, value=1}
         {arriveTangent=-7.0572, interpMode=RCIM_Cubic, leaveTangent=-7.0572, tangentMode=RCTM_User, time=1, value=0}
        ]}}})) {module=ScaleRibbonWidth002}
   &ScaleColor(
    Scale Alpha=&FloatFromCurve(FloatCurve=DataInterface() {propertyValues {Curve {keys [{interpMode=RCIM_Cubic, tangentMode=RCTM_User, time=0.3197, value=0.3476}, <at 1 -1>]}}})
    Scale RGB=&MakeVectorFromLinearColorRGB(Color=&FNTimeOfDayColor(
      Night Color=<rgba 0.1954 0.2978 0.6823 1>
      Morning Color=<rgba 0.7 0.7 0.7 1>
      Day Color=<rgba 0.8 0.8 0.8 1>
      Evening Color=<rgba 0.65 0.65 0.65 1>
     ))
    ScaleA=true
    ScaleRGB=true
   ) {module=ScaleColor002}
  ]
  rendererClasses [&NiagaraRibbonRendererProperties]
  renderers [{rendererClass=&NiagaraRibbonRendererProperties}]
 }
}
summary {
 systemName=NS_WinterWinds_Emote_FootStep_L
 emitters [
  {emitterName=SnowFeet001, rendererClasses [&NiagaraSpriteRendererProperties, &NiagaraMeshRendererProperties]}
  {emitterName=Trail, rendererClasses [&NiagaraRibbonRendererProperties]}
 ]
 userVariables [
  User.bIsFrontEnd {type=NiagaraBool}
  User.bIsFrontEndPreview {type=NiagaraBool}
  User.LeftFoot {type=NiagaraFloat, defaultValue=1, description=""}
  User.RibbonIDOffset {type=NiagaraInt32}
  User.TrailOffset {type=NiagaraFloat, defaultValue=0, description=""}
 ]
}
